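# Unlit alpha-composite Niagara sprite material for the feather effects of the female Monarch glider.
# A grayscale feather texture parameter is shaped by an exponent parameter to form a shared mask.
# Particle color RGB times the mask feeds an emissive adjustment function, and particle alpha is multiplied by the same mask.
# A color vector parameter is present but feeds nothing in the graph.
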Material /BRCosmetics/Gadgets/Assets/VinderTech_GliderChute/Glider_Monarch_Female/FX/Materials/M_Glider_Monarch_Feathers
graph {
 properties {
  bUsedWithNiagaraSprites=true
  blendMode=BLEND_AlphaComposite
  pixelDepthOffsetMode=PDOM_Legacy
  shadingModel=MSM_Unlit
  usedShadingModels=Unlit
 }
 nodes [
  n1 = ParticleColor()
  n2 = TextureSampleParameter2D() {parameterName=TX, samplerType=SAMPLERTYPE_LinearGrayscale, texture=/Game/Effects/Fort_Effects/Textures/Motes/T_Glider_Monarch_Feathers}
  n3 = Power(Base=@n2.R, Exp=ScalarParameter() {defaultValue=1.8, parameterName=Exponent})
  MaterialFunctionCall(Emissive=Multiply(A=@n1.RGB, B=@n3)) {materialFunction=&MF_FX_EmissiveAdjustments}
  Multiply(A=@n1.A, B=@n3)
  VectorParameter() {parameterName=col, defaultValue=<rgba 3 1.785 0.3 0>}
 ]
}
