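# Unlit alpha-composite Niagara sprite material for the flare beam effect on the female Monarch glider.
# A grayscale beam texture is sharpened by an exponent, then softened by depth fade and a squared normal-to-camera dot product that fades edge-on views.
# This mask scales particle color, which feeds the shared FX emissive adjustments function, and separately scales particle alpha.
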Material /BRCosmetics/Gadgets/Assets/VinderTech_GliderChute/Glider_Monarch_Female/FX/Materials/M_Glider_Monarch_Flare_Beam
graph {
 properties {
  bUsedWithNiagaraSprites=true
  blendMode=BLEND_AlphaComposite
  pixelDepthOffsetMode=PDOM_Legacy
  shadingModel=MSM_Unlit
  usedShadingModels=Unlit
 }
 nodes [
  n1 = ParticleColor()
  n2 = TextureSample() {
   samplerType=SAMPLERTYPE_LinearGrayscale
   texture=/BRCosmetics/Gadgets/Assets/VinderTech_GliderChute/Glider_Monarch_Female/FX/Textures/T_Glider_Monarch_Flare_Beam
  }
  n3 = ScalarParameter() {defaultValue=3, parameterName=Exponent}
  n4 = Saturate(None=DotProduct(A=VertexNormalWS(), B=CameraVectorWS()))
  n5 = Multiply(
   A=DepthFade(Opacity=Power(Base=@n2.R, Exp=@n3), FadeDistance=ScalarParameter() {defaultValue=22, parameterName=Depth Fade Distance})
   B=Multiply(A=@n4, B=@n4)
  )
  MaterialFunctionCall(Emissive=Multiply(A=@n1.RGB, B=@n5)) {materialFunction=&MF_FX_EmissiveAdjustments}
  Multiply(A=@n1.A, B=@n5)
 ]
}
