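# Unlit alpha-composite Niagara sprite material for the circular flare effect on the female Monarch glider.
# Summed sines at incommensurate rates warp time, offset by particle random, creating irregular flicker that modulates a power exponent on the flare texture.
# The RGB passes through optional desaturation and particle color tint into an emissive adjustment function, while powered texture alpha multiplies particle alpha.
# The Exponent and col parameters appear unconnected.
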Material /BRCosmetics/Gadgets/Assets/VinderTech_GliderChute/Glider_Monarch_Female/FX/Materials/M_Glider_Monarch_Flare_Circle
graph {
 properties {
  bUsedWithNiagaraSprites=true
  blendMode=BLEND_AlphaComposite
  pixelDepthOffsetMode=PDOM_Legacy
  shadingModel=MSM_Unlit
  usedShadingModels=Unlit
 }
 nodes [
  n1 = ParticleColor()
  n2 = TextureSample() {texture=/BRCosmetics/Gadgets/Assets/VinderTech_GliderChute/Glider_Monarch_Female/FX/Textures/T_Glider_Monarch_Flare_Circular}
  n3 = Constant() {desc=Time bumpiness, r=6}
  n4 = Time()
  n5 = Multiply(A=@n3, B=@n4) {constA=6}
  n6 = Divide() {constA=355, constB=113, desc=Pi}
  n7 = Add(A=@n4, B=Multiply(A=@n3, B=ParticleRandom()) {constB=9.876})
  n8 = Sine(None=Multiply(A=Divide(B=@n6), B=@n7))
  n9 = Sine(None=Multiply(A=Divide(B=@n6) {constA=1.618}, B=@n7))
  n10 = Multiply(A=Divide(B=@n6) {constA=4.235}, B=@n7)
  n11 = Add(A=Add(A=Add(A=@n5, B=@n8), B=Multiply(A=@n9) {constB=0.618}), B=Multiply(A=Sine(None=@n10)) {constB=0.236})
  n12 = Divide(B=@n3)
  n13 = Sine(None=Multiply(A=Multiply(A=@n11, B=@n12) {constB=0.1667}, B=Constant() {desc=Flicker speed, r=6.18}))
  n14 = Power(Base=@n2.RGBA, Exp=LinearInterpolate(Alpha=ConstantBiasScale(None=@n13)) {constA=1.6})
  n15 = ComponentMask(None=@n14) {b=true, g=true, r=true}
  n16 = Desaturation(None=@n15, Fraction=ScalarParameter() {parameterName=Desaturate})
  n17 = ScalarParameter() {parameterName=UseDesaturation}
  MaterialFunctionCall(Emissive=Multiply(A=@n1.RGB, B=LinearInterpolate(A=@n15, B=@n16, Alpha=@n17))) {materialFunction=&MF_FX_EmissiveAdjustments}
  Multiply(A=@n1.A, B=ComponentMask(None=@n14) {a=true})
  OneMinus(None=@n2.RGBA)
  OneMinus(None=@n14)
  ScalarParameter() {defaultValue=1, parameterName=Exponent}
  VectorParameter() {parameterName=col, defaultValue=<rgba 3 1.785 0.3 0>}
 ]
}
