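# Unlit alpha-composite Niagara sprite material for the star flare effect on the female Monarch glider.
# A grayscale star texture is reshaped by inverting it, applying an Exponent parameter and inverting again.
# A saturated dot of vertex normal and camera vector fades that mask as the surface turns from view.
# Particle color RGB times the mask feeds MF_FX_EmissiveAdjustments, while particle alpha is multiplied by the same mask.
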Material /BRCosmetics/Gadgets/Assets/VinderTech_GliderChute/Glider_Monarch_Female/FX/Materials/M_Glider_Monarch_Flare_Star
graph {
 properties {
  bUsedWithNiagaraSprites=true
  blendMode=BLEND_AlphaComposite
  pixelDepthOffsetMode=PDOM_Legacy
  shadingModel=MSM_Unlit
  usedShadingModels=Unlit
 }
 nodes [
  n1 = ParticleColor()
  n2 = TextureSample() {
   samplerType=SAMPLERTYPE_LinearGrayscale
   texture=/BRCosmetics/Gadgets/Assets/VinderTech_GliderChute/Glider_Monarch_Female/FX/Textures/T_Glider_Monarch_Flare_Star
  }
  n3 = OneMinus(None=@n2.R)
  n4 = ScalarParameter() {defaultValue=0.25, parameterName=Exponent}
  n5 = VertexNormalWS()
  n6 = CameraVectorWS()
  n7 = Multiply(A=OneMinus(None=Power(Base=@n3, Exp=@n4)), B=Saturate(None=DotProduct(A=@n5, B=@n6)))
  MaterialFunctionCall(Emissive=Multiply(A=@n1.RGB, B=@n7)) {materialFunction=&MF_FX_EmissiveAdjustments}
  Multiply(A=@n1.A, B=@n7)
 ]
}
