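# Unlit alpha-composite sprite material for Niagara mote particles in the female Monarch glider's effects.
# The grayscale mote texture's red channel is raised to an exponent parameter to shape the sprite mask.
# That mask multiplies particle color, feeding a shared emissive-adjustment function, and separately multiplies particle alpha.
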Material /BRCosmetics/Gadgets/Assets/VinderTech_GliderChute/Glider_Monarch_Female/FX/Materials/M_Glider_Monarch_Motes
graph {
 properties {
  bUsedWithNiagaraSprites=true
  blendMode=BLEND_AlphaComposite
  pixelDepthOffsetMode=PDOM_Legacy
  shadingModel=MSM_Unlit
  usedShadingModels=Unlit
 }
 nodes [
  n1 = ParticleColor()
  n2 = TextureSample() {
   samplerType=SAMPLERTYPE_LinearGrayscale
   texture=/BRCosmetics/Gadgets/Assets/VinderTech_GliderChute/Glider_Monarch_Female/FX/Textures/T_Glider_Monarch_Motes
  }
  n3 = Power(Base=@n2.R, Exp=ScalarParameter() {defaultValue=1.7, parameterName=Exponent})
  MaterialFunctionCall(Emissive=Multiply(A=@n1.RGB, B=@n3)) {materialFunction=&MF_FX_EmissiveAdjustments}
  Multiply(A=@n1.A, B=@n3)
 ]
}
