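# Niagara effect for the female Monarch glider's deploy: a warm gold, sprite-only burst of star flash, flare, beams, waves, motes and feathers.
# Seven independent emitters each spawn instantaneous bursts that scale and fade over life.
# Star, beam and wave sprites share fixed custom facing, with beams and waves fanned evenly by normalized particle count.
# Motes fly outward under drag, while feathers add random acceleration, curl noise and spin.
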NiagaraSystem /BRCosmetics/Gadgets/Assets/VinderTech_GliderChute/Glider_Monarch_Female/FX/NS_Monarch_Female_Glider_Deploy
dependencies {
 usedDataInterfaces [&NiagaraDataInterfaceColorCurve, &NiagaraDataInterfaceCurve, &NiagaraDataInterfaceVector2DCurve]
 usedDynamicInputs []
 usedModules [&SystemState, &FN_Integrator]
}
emitters {
 Feathers_Blast {
  emitterUpdateScript [
   &EmitterState()
   &SpawnBurst_Instantaneous(Spawn Count=16, Spawn Time=0.05, Use Spawn Probability=false, UseLoopCountLimit=false)
  ]
  particleSpawnScript [
   &InitializeParticle(
    Color=<rgba 1.5 1.1263 0.525 1>
    Lifetime Max=1.333
    Lifetime Min=0.7
    Lifetime Mode=::NewEnumerator1
    Sprite Rotation Mode=::NewEnumerator2
    Sprite Size Mode=::NewEnumerator1
    Uniform Sprite Size Max=19
    Uniform Sprite Size Min=7
    UsePositionOffset=false
    Write Material Random=true
   )
   &SphereLocation(
    MaskSpawnGroup=false
    Override Seed=false
    Sphere Radius=1
    Surface Only=false
    UseNonUniformScale=false
    UseOffset=false
    UseSphereOrientation=false
   )
   &SubUVAnimation(End Frame=8, SubUV Animation Mode:ENiagara_SubUVLookupMode=::NewEnumerator2)
   &AddVelocityFromPoint(
    Module.Use Falloff Distance=false
    Origin Offset=<0 0 0.5>
    Velocity Strength=&RandomRangeFloat(Minimum=222, Maximum=444)
   )
   &AddVelocity(Velocity=&MakeVector(X=&RandomRangeFloat(Minimum=-555, Maximum=555), Y=0, Z=0))
  ]
  particleUpdateScript [
   &Drag(Drag=3, Use Linear Drag=true, Use Rotational Drag=false)
   &AccelerationForce(Acceleration=&Multiply_VectorByFloat(
     Vector=&MakeVector(
      X=&RandomRangeFloat(Minimum=-999, Maximum=999)
      Y=0
      Z=&RandomRangeFloat(Minimum=333, Maximum=999)
     )
     Float=&FloatFromCurve(FloatCurve=DataInterface() {propertyValues {Curve {keys [<at 0 0 RCIM_Constant>,<at 0.2 1>]}}})
    ))
   &CurlNoiseForce(Noise Frequency=100, Noise Strength=500, Pan Noise=false)
   &SolveForcesAndVelocity(Clamp Velocity=false, Limit Acceleration=false, Speed Limit=666)
   &ParticleState()
   &ScaleColor(
    Scale Mode=::NewEnumerator1
    Scale RGBA=&MakeVector4_FromFloat(Float=&FloatFromCurve(FloatCurve=DataInterface() {propertyValues {Curve {keys [
          {arriveTangent=13.1284, interpMode=RCIM_Cubic, leaveTangent=13.1284, tangentMode=RCTM_User, time=0, value=0}
          {arriveTangent=-3.0086, interpMode=RCIM_Cubic, leaveTangent=-3.0086, tangentMode=RCTM_User, time=0.1, value=1}
          <at 1 0>
         ]}}}))
    ScaleRGBA=true
   )
   &SpriteRotationRate(Rotation Rate=&RandomRangeFloat(Minimum=-666, Maximum=666))
   &ScaleSpriteSize(Uniform Curve Sprite Scale=DataInterface() {propertyValues {Curve {keys [
        {interpMode=RCIM_Cubic, tangentMode=RCTM_User, time=0, value=1}
        {arriveTangent=-3.2205, interpMode=RCIM_Cubic, leaveTangent=-3.2205, tangentMode=RCTM_User, time=1, value=0}
       ]}}})
  ]
  rendererClasses [&NiagaraSpriteRendererProperties]
  renderers [
   {rendererClass=&NiagaraSpriteRendererProperties}
   {rendererClass=&NiagaraSpriteRendererProperties, rendererIndex=1}
  ]
 }
 Flare {
  emitterUpdateScript [
   &EmitterState()
   &SpawnBurst_Instantaneous(Use Spawn Probability=false, UseLoopCountLimit=false)
  ]
  particleSpawnScript [&InitializeParticle(
    Color=/Niagara/DynamicInputs/TypeConversions/MakeLinearColorFromFloat(Value:NiagaraFloat=0.6)
    Lifetime=0.7
    Sprite Rotation Mode=::NewEnumerator2
    Sprite Size Mode=::NewEnumerator3
    Uniform Sprite Size=150
    UsePositionOffset=false
    Write Material Random=true
   )]
  particleUpdateScript [
   &ParticleState()
   &ScaleColor(
    Scale Mode=::NewEnumerator1
    Scale RGBA=&MakeVector4_FromFloat(Float=&FloatFromCurve(FloatCurve=DataInterface() {propertyValues {Curve {keys [
          <at 0 0>
          {arriveTangent=-3.7291, interpMode=RCIM_Cubic, leaveTangent=-3.7291, tangentMode=RCTM_User, time=0.05, value=1}
          {arriveTangentWeight=1.0122, interpMode=RCIM_Cubic, tangentMode=RCTM_User, time=1, value=0}
         ]}}}))
    ScaleRGBA=true
   )
   &ScaleSpriteSize(Uniform Curve Sprite Scale=DataInterface() {propertyValues {Curve {keys [
        {arriveTangent=2.4947, interpMode=RCIM_Cubic, leaveTangent=2.4947, tangentMode=RCTM_User, time=0, value=0.333}
        {arriveTangent=0.503, interpMode=RCIM_Cubic, leaveTangent=0.503, tangentMode=RCTM_User, time=0.4, value=1.25}
        {tangentMode=RCTM_User, time=1, value=1.333}
       ]}}})
   &SpriteRotationRate(Rotation Rate=0.2)
  ]
  rendererClasses [&NiagaraSpriteRendererProperties]
  renderers [{rendererClass=&NiagaraSpriteRendererProperties}]
 }
 Horizontal_Beams {
  emitterUpdateScript [
   &EmitterState()
   &SpawnBurst_Instantaneous(Spawn Count=2, Spawn Time=0.1, Use Spawn Probability=false, UseLoopCountLimit=false)
  ]
  particleSpawnScript [&InitializeParticle(
    Color=<rgba 1.5 1 0.3 1.5>
    Position Offset=<0 4 0>
    Sprite Size=<144 333>
    Sprite Size Mode=::NewEnumerator4
    UsePositionOffset=true
    Write Material Random=true
   )]
  particleUpdateScript [
   &ParticleState()
   &ScaleColor(
    Scale Mode=::NewEnumerator1
    Scale RGBA=&MakeVector4_FromFloat(Float=&FloatFromCurve(FloatCurve=DataInterface() {propertyValues {Curve {keys [<at 0.05 0>, {arriveTangent=-3.0726, interpMode=RCIM_Cubic, leaveTangent=-3.0726, tangentMode=RCTM_User, time=0.1, value=1}, <at 1 0>]}}}))
    ScaleRGBA=true
   )
   /SetVariables(Particles.SpriteRotation:NiagaraFloat=&Multiply_FloatByInt(Float=&Add_Float(A=/Game/Effects/Niagara/DynamicInputs/Math/FN_NormalizedParticleCount(), B=0.25), Integer=360)) {isSetParameters=true, module=SetVariables_#1}
   &ScaleSpriteSize(
    Non-Uniform Curve Sprite Scale=DataInterface() {propertyValues {
      XCurve {keys [{tangentMode=RCTM_User, time=0, value=3}, {tangentMode=RCTM_User, time=1, value=3}]}
      YCurve {keys [
        {arriveTangent=5.5646, interpMode=RCIM_Cubic, leaveTangent=5.5646, tangentMode=RCTM_User, time=0, value=0}
        {tangentMode=RCTM_User, time=1, value=1.5}
       ]}
     }}
    Scale Sprite Size Mode=::NewEnumerator3
   )
   &SpriteFacingAndAlignment(Sprite Alignment=<0 0 1>, Sprite Facing=<0 -1 0.1>, WriteAlignment=true, WriteFacing=true)
  ]
  rendererClasses [&NiagaraSpriteRendererProperties]
  renderers [{rendererClass=&NiagaraSpriteRendererProperties}]
 }
 Horizontal_Waves {
  emitterUpdateScript [
   &EmitterState()
   &SpawnBurst_Instantaneous(Spawn Count=2, Spawn Time=0.4, Use Spawn Probability=false, UseLoopCountLimit=false)
  ]
  particleSpawnScript [&InitializeParticle(
    Color=<rgba 0.4 0.2667 0.08 0.4>
    Lifetime=0.5
    Sprite Size=<111 275>
    Sprite Size Mode=::NewEnumerator4
    UsePositionOffset=false
    Write Material Random=true
   )]
  particleUpdateScript [
   &ParticleState()
   &ScaleColor(
    Scale Mode=::NewEnumerator1
    Scale RGBA=&MakeVector4_FromFloat(Float=&FloatFromCurve(FloatCurve=DataInterface() {propertyValues {Curve {keys [<at 0 0>, {arriveTangent=-3.0726, interpMode=RCIM_Cubic, leaveTangent=-3.0726, tangentMode=RCTM_User, time=0.05, value=1}, <at 1 0>]}}}))
    ScaleRGBA=true
   )
   /SetVariables(Particles.SpriteRotation:NiagaraFloat=&Multiply_FloatByInt(Float=&Add_Float(A=&FN_NormalizedParticleCount(), B=0.25), Integer=360)) {isSetParameters=true, module=SetVariables_#1}
   &ScaleSpriteSize(
    Non-Uniform Curve Sprite Scale=DataInterface() {propertyValues {
      XCurve {keys [
        {arriveTangent=6.2067, interpMode=RCIM_Cubic, leaveTangent=6.2067, tangentMode=RCTM_User, time=0, value=0}
        {tangentMode=RCTM_User, time=1, value=6}
       ]}
      YCurve {keys [
        {arriveTangent=2.8894, interpMode=RCIM_Cubic, leaveTangent=2.8894, tangentMode=RCTM_User, time=0, value=0}
        {tangentMode=RCTM_User, time=1, value=1.5}
       ]}
     }}
    Scale Sprite Size Mode=::NewEnumerator3
   )
   &CameraOffset(Camera Offset Amount=55)
   &SpriteFacingAndAlignment(Sprite Alignment=<0 0 1>, Sprite Facing=<0 -1 0.1>, WriteAlignment=true, WriteFacing=true)
  ]
  rendererClasses [&NiagaraSpriteRendererProperties]
  renderers [{rendererClass=&NiagaraSpriteRendererProperties}]
 }
 Motes_Blast {
  emitterUpdateScript [
   &EmitterState()
   &SpawnBurst_Instantaneous(Spawn Count=64, Use Spawn Probability=false, UseLoopCountLimit=false)
  ]
  particleSpawnScript [
   &InitializeParticle(
    Color=<rgba 1.5 1.1263 0.525 1>
    Lifetime Max=0.7
    Lifetime Min=0.4
    Lifetime Mode=::NewEnumerator1
    Sprite Size Mode=::NewEnumerator1
    Uniform Sprite Size Max=4
    Uniform Sprite Size Min=2
    UsePositionOffset=false
    Write Material Random=true
   )
   &SphereLocation(
    MaskSpawnGroup=false
    Override Seed=false
    Sphere Radius=1
    Surface Only=false
    UseNonUniformScale=false
    UseOffset=false
    UseSphereOrientation=false
   )
   &SubUVAnimation(End Frame=8, SubUV Animation Mode:ENiagara_SubUVLookupMode=::NewEnumerator2)
   &AddVelocityFromPoint(Module.Use Falloff Distance=false, Velocity Strength=&RandomRangeFloat(Minimum=150, Maximum=450))
  ]
  particleUpdateScript [
   &Drag(Drag=4, Use Linear Drag=true, Use Rotational Drag=false)
   &SolveForcesAndVelocity(Clamp Velocity=false, Limit Acceleration=false)
   &ParticleState()
   &ScaleColor(
    Scale Mode=::NewEnumerator1
    Scale RGBA=&MakeVector4_FromFloat(Float=&FloatFromCurve(FloatCurve=DataInterface() {propertyValues {Curve {keys [
          {arriveTangent=13.1284, interpMode=RCIM_Cubic, leaveTangent=13.1284, tangentMode=RCTM_User, time=0, value=0}
          {arriveTangent=-3.0086, interpMode=RCIM_Cubic, leaveTangent=-3.0086, tangentMode=RCTM_User, time=0.1, value=1}
          <at 1 0>
         ]}}}))
    ScaleRGBA=true
   )
  ]
  rendererClasses [&NiagaraSpriteRendererProperties]
  renderers [{rendererClass=&NiagaraSpriteRendererProperties}]
 }
 Star {
  emitterUpdateScript [
   &EmitterState()
   &SpawnBurst_Instantaneous(Use Spawn Probability=false, UseLoopCountLimit=false)
  ]
  particleSpawnScript [&InitializeParticle(
    Color=<rgba 13 8.6667 2.6 1>
    Lifetime=0.9
    Position Offset=<0 7 0>
    Sprite Size Mode=::NewEnumerator3
    Uniform Sprite Size=22
    UsePositionOffset=true
    Write Material Random=true
   )]
  particleUpdateScript [
   &ParticleState()
   &ScaleColor(
    Scale Mode=::NewEnumerator1
    Scale RGBA=&MakeVector4_FromFloat(Float=&FloatFromCurve(FloatCurve=DataInterface() {propertyValues {Curve {keys [<at 0 22>, {arriveTangent=-22, leaveTangent=-22, time=0.075, value=0.5}, <at 1 0>]}}}))
    ScaleRGBA=true
   )
   &ScaleSpriteSize(Uniform Curve Sprite Scale=DataInterface() {propertyValues {Curve {keys [
        {arriveTangent=10.5825, interpMode=RCIM_Cubic, leaveTangent=10.5825, tangentMode=RCTM_User, time=0, value=0}
        {arriveTangent=3.0843, interpMode=RCIM_Cubic, leaveTangent=1.71, tangentMode=RCTM_Break, time=0.1, value=0.55}
        {tangentMode=RCTM_User, time=1, value=1}
       ]}}})
   &SpriteFacingAndAlignment(Sprite Alignment=<0 0 1>, Sprite Facing=<0 -1 0.1>, WriteAlignment=true, WriteFacing=true)
  ]
  rendererClasses [&NiagaraSpriteRendererProperties]
  renderers [{rendererClass=&NiagaraSpriteRendererProperties}]
 }
 Star_Beams {
  emitterUpdateScript [
   &EmitterState()
   &SpawnBurst_Instantaneous(Spawn Count=5, Use Spawn Probability=false, UseLoopCountLimit=false)
  ]
  particleSpawnScript [&InitializeParticle(
    Color=<rgba 13 8.6667 2.6 3>
    Lifetime=0.75
    Position Offset=<0 7 0>
    Sprite Size Mode=::NewEnumerator3
    Uniform Sprite Size=33
    UsePositionOffset=true
    Write Material Random=true
   )]
  particleUpdateScript [
   &ParticleState()
   &ScaleColor(
    Scale Mode=::NewEnumerator1
    Scale RGBA=&MakeVector4_FromFloat(Float=&FloatFromCurve(FloatCurve=DataInterface() {propertyValues {Curve {keys [<at 0 0>, {arriveTangent=-3.0726, interpMode=RCIM_Cubic, leaveTangent=-3.0726, tangentMode=RCTM_User, time=0.1, value=1}, <at 1 0>]}}}))
    ScaleRGBA=true
   )
   /SetVariables(Particles.SpriteRotation:NiagaraFloat=&Multiply_FloatByInt(Float=&FN_NormalizedParticleCount(), Integer=360)) {isSetParameters=true, module=SetVariables_#1}
   &ScaleSpriteSize(
    Non-Uniform Curve Sprite Scale=DataInterface() {propertyValues {
      XCurve {keys [{tangentMode=RCTM_User, time=0, value=3}, {tangentMode=RCTM_User, time=1, value=0}]}
      YCurve {keys [{tangentMode=RCTM_User, time=0, value=1}, {tangentMode=RCTM_User, time=1, value=7}]}
     }}
    Scale Sprite Size Mode=::NewEnumerator3
   )
   &SpriteFacingAndAlignment(Sprite Alignment=<0 0 1>, Sprite Facing=<0 -1 0.1>, WriteAlignment=true, WriteFacing=true)
  ]
  rendererClasses [&NiagaraSpriteRendererProperties]
  renderers [{rendererClass=&NiagaraSpriteRendererProperties}]
 }
}
summary {
 systemName=NS_Monarch_Female_Glider_Deploy
 emitters [
  {emitterName=Star, rendererClasses [&NiagaraSpriteRendererProperties]}
  {emitterName=Flare, rendererClasses [&NiagaraSpriteRendererProperties]}
  {emitterName=Star_Beams, rendererClasses [&NiagaraSpriteRendererProperties]}
  {emitterName=Horizontal_Beams, rendererClasses [&NiagaraSpriteRendererProperties]}
  {emitterName=Horizontal_Waves, rendererClasses [&NiagaraSpriteRendererProperties]}
  {emitterName=Motes_Blast, rendererClasses [&NiagaraSpriteRendererProperties]}
  {emitterName=Feathers_Blast, rendererClasses [&NiagaraSpriteRendererProperties]}
 ]
 userVariables [
  User.bIsFrontEnd {type=NiagaraBool}
  User.bIsFrontEndPreview {type=NiagaraBool}
  User.bIsFullyDeployed {type=NiagaraBool}
  User.ForwardDot {type=NiagaraFloat, defaultValue=0, description=""}
  User.RibbonIDOffset {type=NiagaraInt32}
  User.RightDot {type=NiagaraFloat, defaultValue=0, description=""}
  User.SkeletalMesh {
   type=NiagaraDataInterfaceSkeletalMesh
   defaultValue {
    struct=NiagaraExt_VariableValue_DataInterface
    value {dataInterface=&self:NiagaraDataInterfaceSkeletalMesh_0}
   }
  }
 ]
}
